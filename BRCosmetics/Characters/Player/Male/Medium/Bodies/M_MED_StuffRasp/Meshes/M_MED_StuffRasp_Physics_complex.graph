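# Physics asset for the M_MED_StuffRasp male medium player body, covering its hanging thigh chains and the body parts around them.
# Capsules approximate pelvis, spine, thighs and calves, while smaller capsules form front and back chains on each side.
# Angle-limited constraints link chain segments to tiny BASE anchor bodies, except the right front chain, which attaches to spinal_chain_front_02_m.
# Free-rotation constraints join front and back chain ends into a span per side.
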PhysicsAsset /BRCosmetics/Characters/Player/Male/Medium/Bodies/M_MED_StuffRasp/Meshes/M_MED_StuffRasp_Physics_complex
bodies {
 BASE_thigh_chain_back_02_l1 [{shapeType=Sphere, radius=0.505, shapeName=BASE_thigh_chain_back_02_l1_sphere}]
 BASE_thigh_chain_back_02_r1 [{shapeType=Capsule, length=0.505, radius=0.505, shapeName=BASE_thigh_chain_back_02_r1_capsule, rotation=<rot 0 -90 0>}]
 BASE_thigh_chain_front_01_l [{shapeType=Capsule, length=0.505, radius=0.505, shapeName=BASE_thigh_chain_front_01_l_capsule, rotation=<rot 0 -90 0>}]
 calf_l [{shapeType=Capsule, length=25.2327, radius=8.7367, center=<-31.3423 0.2534 -0.0985>, rotation=<rot 0.004 -89.5027 -89.9452>}]
 calf_r [{shapeType=Capsule, length=26.7725, radius=7.1969, center=<33.1105 -3.2323 0.0971>, rotation=<rot 0.004 -89.5027 90.0548>}]
 pelvis [
  {shapeType=Capsule, length=14.0931, radius=9.5578, shapeName=pelvis_capsule, center=<-7.2206 -0.7911 0>}
  {shapeType=Capsule, length=14.0931, radius=9.5578, shapeName=pelvis_capsule, center=<-17.1354 -0.7911 0>, rotation=<rot 90 0 0>}
 ]
 spinal_chain_front_02_m [{
   shapeType=Capsule
   length=1.1923
   radius=1.55
   shapeName=spinal_chain_front_02_m_capsule
   center=<-0.3413 -0.0412 0.0231>
   rotation=<rot 0 -88.2844 -88.8768>
  }]
 spine_01 [{shapeType=Capsule, length=8.9504, radius=13.1879, center=<0.385 0.0779 0.0468>, rotation=<rot -2.366 -90 -0.0656>}]
 spine_03 [{shapeType=Capsule, length=9.9533, radius=11.3724, center=<-0.8119 -0.2609 0.1138>, rotation=<rot 0.1383 -90 -0.6386>}]
 spine_04 [{shapeType=Capsule, length=7.1469, radius=13.0939}]
 spine_05 [{shapeType=Capsule, length=13.5645, radius=9.9479, center=<1.1175 0.1083 -0.9321>, rotation=<rot 1.12 -90 -0.7413>}]
 thigh_chain_back_02_l [{
   shapeType=Capsule
   length=5.8689
   radius=2.8761
   shapeName=thigh_chain_back_02_l_capsule
   center=<6.2427 -1.0688 -4.0268>
   rotation=<rot 7.9152 -89.4054 -89.0075>
  }]
 thigh_chain_back_02_l1 [{
   shapeType=Capsule
   length=9.4488
   radius=2.9428
   shapeName=thigh_chain_back_02_l1_capsule
   center=<7.47 -0.1488 -0.8942>
   rotation=<rot 0 88.822 -90.9444>
  }]
 thigh_chain_back_02_r [{
   shapeType=Capsule
   length=8.1049
   radius=2.8761
   shapeName=thigh_chain_back_02_l_capsule
   center=<7.6033 -3.4594 -2.1126>
   rotation=<rot 7.4462 -117.7198 -43.1762>
  }]
 thigh_chain_back_02_r1 [{
   shapeType=Capsule
   length=11.4848
   radius=2.9428
   shapeName=thigh_chain_back_02_l1_capsule
   center=<9.2157 -0.1847 0.8818>
   rotation=<rot 0 88.822 -90.9444>
  }]
 thigh_chain_front_01_l [{
   shapeType=Capsule
   length=5.8989
   radius=1.5306
   shapeName=thigh_chain_front_01_l_capsule
   center=<3.3504 -0.0421 -0.1237>
   rotation=<rot 0.0001 90.0185 -95.3818>
  }]
 thigh_chain_front_01_r [{
   shapeType=Capsule
   length=5.8989
   radius=1.5306
   shapeName=thigh_chain_front_01_l_capsule
   center=<3.3504 -0.0421 -0.1237>
   rotation=<rot 0.0001 90.0185 -95.3818>
  }]
 thigh_chain_front_02_l [{
   shapeType=Capsule
   length=5.9929
   radius=1.4366
   shapeName=thigh_chain_front_01_l_capsule
   center=<3.2077 -0.1009 0.3051>
   rotation=<rot 0.2887 87.0526 -84.4471>
  }]
 thigh_chain_front_02_r [{
   shapeType=Capsule
   length=9.0454
   radius=1.4366
   shapeName=thigh_chain_front_01_l_capsule
   center=<4.4382 -0.1648 0.4249>
   rotation=<rot 0.2749 87.0481 -91.2795>
  }]
 thigh_chain_front_03_l [{
   shapeType=Capsule
   length=9.1246
   radius=1.7863
   shapeName=thigh_chain_front_01_l_capsule
   center=<7.0179 -0.0409 0.9759>
   rotation=<rot 0 90.0185 -98.7252>
  }]
 thigh_chain_front_03_r [
  {
   shapeType=Capsule
   length=5.2676
   radius=1.7863
   shapeName=thigh_chain_front_01_l_capsule
   center=<8.9439 -0.0407 -0.5345>
   rotation=<rot -0.0096 90.0125 -133.3285>
  }
  {
   shapeType=Capsule
   length=5.7004
   radius=1.7863
   shapeName=thigh_chain_front_03_r_capsule
   center=<4.2593 -0.0436 1.5464>
   rotation=<rot -0.0096 90.0125 -93.5076>
  }
 ]
 thigh_l_extra [{shapeType=Capsule, length=24.182, radius=9.7874, center=<-20.9364 -0.7071 0.8078>, rotation=<rot 0.004 -89.5027 90.0548>}]
 thigh_twist_01_r [
  {shapeType=Capsule, length=4.5864, radius=8.2148, center=<-11.1478 0.4287 -0.4552>, rotation=<rot 0.004 -89.5027 -89.9452>}
  {
   shapeType=Capsule
   length=16.3585
   radius=9.1831
   shapeName=thigh_twist_01_r_capsule
   center=<2.5813 0.461 -0.4588>
   rotation=<rot 0.004 -89.5027 -89.9452>
  }
 ]
}
constraints [
 {bone1Name=thigh_chain_front_03_l, bone2Name=thigh_chain_front_02_l, swing1LimitDegrees=40, swing2LimitDegrees=40, twistLimitDegrees=10}
 {bone1Name=thigh_chain_back_02_l, bone2Name=thigh_chain_back_02_l1, swing1LimitDegrees=20, swing2LimitDegrees=15, twistLimitDegrees=10}
 {
  bone1Name=spinal_chain_front_02_m
  bone2Name=pelvis
  swing1LimitDegrees=45
  swing1Motion=Free
  swing2LimitDegrees=45
  swing2Motion=Free
  twistLimitDegrees=45
  twistMotion=Free
 }
 {
  bone1Name=thigh_chain_back_02_l
  bone2Name=thigh_chain_front_03_l
  swing1LimitDegrees=45
  swing1Motion=Free
  swing2LimitDegrees=45
  swing2Motion=Free
  twistLimitDegrees=45
  twistMotion=Free
 }
 {bone1Name=thigh_chain_back_02_l1, bone2Name=BASE_thigh_chain_back_02_l1, swing1LimitDegrees=40, swing2LimitDegrees=40, twistLimitDegrees=10}
 {bone1Name=thigh_chain_front_03_r, bone2Name=thigh_chain_front_02_r, swing1LimitDegrees=40, swing2LimitDegrees=40, twistLimitDegrees=10}
 {bone1Name=thigh_chain_back_02_r1, bone2Name=BASE_thigh_chain_back_02_r1, swing1LimitDegrees=40, swing2LimitDegrees=40, twistLimitDegrees=10}
 {bone1Name=thigh_chain_back_02_r, bone2Name=thigh_chain_back_02_r1, swing1LimitDegrees=20, swing2LimitDegrees=15, twistLimitDegrees=10}
 {
  bone1Name=thigh_chain_front_03_r
  bone2Name=thigh_chain_back_02_r
  swing1LimitDegrees=45
  swing1Motion=Free
  swing2LimitDegrees=45
  swing2Motion=Free
  twistLimitDegrees=45
  twistMotion=Free
 }
 {bone1Name=thigh_chain_front_02_l, bone2Name=thigh_chain_front_01_l, swing1LimitDegrees=40, swing2LimitDegrees=40, twistLimitDegrees=10}
 {bone1Name=thigh_chain_front_01_l, bone2Name=BASE_thigh_chain_front_01_l, swing1LimitDegrees=60, swing2LimitDegrees=40, twistLimitDegrees=10}
 {
  bone1Name=thigh_l_extra
  bone2Name=pelvis
  swing1LimitDegrees=45
  swing1Motion=Free
  swing2LimitDegrees=45
  swing2Motion=Free
  twistLimitDegrees=45
  twistMotion=Free
 }
 {bone1Name=thigh_chain_front_02_r, bone2Name=thigh_chain_front_01_r, swing1LimitDegrees=40, swing2LimitDegrees=40, twistLimitDegrees=10}
 {bone1Name=thigh_chain_front_01_r, bone2Name=spinal_chain_front_02_m, swing1LimitDegrees=60, swing2LimitDegrees=40, twistLimitDegrees=10}
 {
  bone1Name=calf_l
  bone2Name=pelvis
  swing1LimitDegrees=45
  swing1Motion=Free
  swing2LimitDegrees=45
  swing2Motion=Free
  twistLimitDegrees=45
  twistMotion=Free
 }
 {
  bone1Name=calf_r
  bone2Name=pelvis
  swing1LimitDegrees=45
  swing1Motion=Free
  swing2LimitDegrees=45
  swing2Motion=Free
  twistLimitDegrees=45
  twistMotion=Free
 }
]
