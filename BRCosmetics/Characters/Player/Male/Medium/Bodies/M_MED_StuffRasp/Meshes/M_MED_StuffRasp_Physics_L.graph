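# Physics asset for the left-side thigh chains on the StuffRasp medium male body cosmetic.
# Pelvis and left thigh capsules sit alongside four chain capsules and three small BASE bodies, with no constraints on the pelvis or thigh.
# Constraints link the back BASE sphere through both back chain bodies to the two front ones.
# The back-to-front joint has free swing and twist, while the other joints set angular limits.
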PhysicsAsset /BRCosmetics/Characters/Player/Male/Medium/Bodies/M_MED_StuffRasp/Meshes/M_MED_StuffRasp_Physics_L
bodies {
 BASE_thigh_chain_back_02_l1 [{shapeType=Sphere, radius=0.505, shapeName=BASE_thigh_chain_back_02_l1_sphere}]
 BASE_thigh_chain_back_02_r1 [{shapeType=Capsule, length=0.505, radius=0.505, shapeName=BASE_thigh_chain_back_02_r1_capsule, rotation=<rot 0 -90 0>}]
 BASE_thigh_chain_front_02_r [{shapeType=Capsule, length=0.505, radius=0.505, shapeName=BASE_thigh_chain_front_02_r_capsule, rotation=<rot 0 -90 0>}]
 pelvis [{shapeType=Capsule, length=10.096, radius=9.5578, shapeName=pelvis_capsule, center=<-7.2206 -0.7911 0>}]
 thigh_chain_back_02_l [{
   shapeType=Capsule
   length=5.8689
   radius=2.8761
   shapeName=thigh_chain_back_02_l_capsule
   center=<6.2427 -1.0688 -4.0268>
   rotation=<rot 7.9152 -89.4054 -89.0075>
  }]
 thigh_chain_back_02_l1 [{
   shapeType=Capsule
   length=4.8331
   radius=2.9428
   shapeName=thigh_chain_back_02_l1_capsule
   center=<4.9983 -0.098 -0.8534>
   rotation=<rot 0 88.822 -90.9444>
  }]
 thigh_chain_front_02_l [{
   shapeType=Capsule
   length=4.9122
   radius=2.5173
   shapeName=thigh_chain_front_01_l_capsule
   center=<3.1227 -0.0422 1.2093>
   rotation=<rot 0.0001 90.0185 -84.4397>
  }]
 thigh_chain_front_03_l [{
   shapeType=Capsule
   length=7.6243
   radius=3.2865
   shapeName=thigh_chain_front_01_l_capsule
   center=<7.2851 -0.0408 2.7241>
   rotation=<rot 0 90.0185 -95.5912>
  }]
 thigh_l [{shapeType=Capsule, length=25.5171, radius=8.3691, center=<-20.9364 -0.7071 0.8078>, rotation=<rot 0.004 -89.5027 90.0548>}]
}
constraints [
 {bone1Name=thigh_chain_front_03_l, bone2Name=thigh_chain_front_02_l, swing1LimitDegrees=40, swing2LimitDegrees=40, twistLimitDegrees=10}
 {bone1Name=thigh_chain_back_02_l, bone2Name=thigh_chain_back_02_l1, swing1LimitDegrees=20, swing2LimitDegrees=15, twistLimitDegrees=10}
 {
  bone1Name=thigh_chain_back_02_l
  bone2Name=thigh_chain_front_03_l
  swing1LimitDegrees=45
  swing1Motion=Free
  swing2LimitDegrees=45
  swing2Motion=Free
  twistLimitDegrees=45
  twistMotion=Free
 }
 {bone1Name=thigh_chain_back_02_l1, bone2Name=BASE_thigh_chain_back_02_l1, swing1LimitDegrees=40, swing2LimitDegrees=40, twistLimitDegrees=10}
]
