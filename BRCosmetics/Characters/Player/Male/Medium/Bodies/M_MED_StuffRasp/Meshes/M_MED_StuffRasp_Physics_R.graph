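# Physics asset for the right-side thigh chain on the StuffRasp medium male body cosmetic.
# Front and back strands each join two capsules to a small BASE anchor body through angle-limited swing and twist constraints.
# A constraint with free swing and twist links the two strand ends, closing the chain into a loop between the anchors.
# Pelvis and right thigh capsules model the body; they and two left-side BASE spheres have no constraints.
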PhysicsAsset /BRCosmetics/Characters/Player/Male/Medium/Bodies/M_MED_StuffRasp/Meshes/M_MED_StuffRasp_Physics_R
bodies {
 BASE_thigh_chain_back_02_l1 [{shapeType=Sphere, radius=0.505, shapeName=BASE_thigh_chain_back_02_l1_sphere}]
 BASE_thigh_chain_back_02_r1 [{shapeType=Capsule, length=0.505, radius=0.505, shapeName=BASE_thigh_chain_back_02_r1_capsule, rotation=<rot 0 -90 0>}]
 BASE_thigh_chain_front_02_l [{shapeType=Sphere, radius=0.505, shapeName=BASE_thigh_chain_front_02_l_sphere}]
 BASE_thigh_chain_front_02_r [{shapeType=Capsule, length=0.505, radius=0.505, shapeName=BASE_thigh_chain_front_02_r_capsule, rotation=<rot 0 -90 0>}]
 pelvis [{shapeType=Capsule, length=10.096, radius=9.5578, shapeName=pelvis_capsule, center=<-7.2206 -0.7911 0>}]
 thigh_chain_back_02_r [{
   shapeType=Capsule
   length=5.8689
   radius=2.8761
   shapeName=thigh_chain_back_02_l_capsule
   center=<7.0181 -2.932 -0.6232>
   rotation=<rot 7.4461 -117.7198 -54.3913>
  }]
 thigh_chain_back_02_r1 [{
   shapeType=Capsule
   length=9.0835
   radius=2.9428
   shapeName=thigh_chain_back_02_l1_capsule
   center=<8.0005 -0.1597 0.9019>
   rotation=<rot 0 88.822 -90.9444>
  }]
 thigh_chain_front_02_r [{
   shapeType=Capsule
   length=4.9122
   radius=2.5173
   shapeName=thigh_chain_front_01_l_capsule
   center=<3.1227 -0.0422 1.2093>
   rotation=<rot 0.0001 90.0185 -84.4397>
  }]
 thigh_chain_front_03_r [{
   shapeType=Capsule
   length=7.6243
   radius=3.2865
   shapeName=thigh_chain_front_01_l_capsule
   center=<7.0729 1.8167 2.0122>
   rotation=<rot 0 90.0185 -106.556>
  }]
 thigh_r [{shapeType=Capsule, length=24.182, radius=9.7874, center=<20.9364 0.7071 -0.8078>, rotation=<rot 0.004 -89.5027 -89.9452>}]
}
constraints [
 {bone1Name=thigh_chain_front_02_r, bone2Name=BASE_thigh_chain_front_02_r, swing1LimitDegrees=20, swing2LimitDegrees=15, twistLimitDegrees=10}
 {bone1Name=thigh_chain_front_03_r, bone2Name=thigh_chain_front_02_r, swing1LimitDegrees=40, swing2LimitDegrees=40, twistLimitDegrees=10}
 {bone1Name=thigh_chain_back_02_r1, bone2Name=BASE_thigh_chain_back_02_r1, swing1LimitDegrees=40, swing2LimitDegrees=40, twistLimitDegrees=10}
 {bone1Name=thigh_chain_back_02_r, bone2Name=thigh_chain_back_02_r1, swing1LimitDegrees=20, swing2LimitDegrees=15, twistLimitDegrees=10}
 {
  bone1Name=thigh_chain_front_03_r
  bone2Name=thigh_chain_back_02_r
  swing1LimitDegrees=45
  swing1Motion=Free
  swing2LimitDegrees=45
  swing2Motion=Free
  twistLimitDegrees=45
  twistMotion=Free
 }
]
