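# Animation Blueprint for the StuffRasp male character's face accessory part, adding procedural motion to its hood.
# It copies the attached parent's pose, then layers in the Body_Mesh pose at clavicle and pectoral bones using mesh-space rotation blending.
# A part-specific Control Rig follows, then identically tuned spring-based AnimDynamics nodes simulate the left and right hood bones.
# Body_Mesh is set from the custom mesh parts when character customization loads.
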AnimBlueprint /BRCosmetics/Characters/Player/Male/Medium/Bodies/M_MED_StuffRasp/Meshes/Parts/M_MED_StuffRasp_FaceAcc_AnimBP
target_skeleton=&/M_MED_StuffRasp_FaceAcc_Skeleton
graphs [
 AnimGraph {nodes [
   n1 = Layered blend per bone(
    BasePose=Copy Pose From Mesh() {bUseAttachedParent=true}
    BlendPoses_0=Copy Pose From Mesh(SourceMeshComponent=Get Body_Mesh()) {bUseAttachedParent=true}
   ) {
    bMeshSpaceRotationBlend=true
    layerSetup [{branchFilters [
       {blendDepth=3, boneName=clavicle_l}
       {blendDepth=3, boneName=clavicle_r}
       {blendDepth=3, boneName=deform_pec_aim_base_l}
       {blendDepth=3, boneName=deform_pec_aim_base_r}
      ]}]
   }
   n2 = Control Rig(Source=@n1) {controlRigAssetReference {blueprintRigClass=&/M_MED_StuffRasp_FaceAcc_CtrlRig.M_MED_StuffRasp_FaceAcc_CtrlRig_C}}
   n3 = Anim Dynamics - Bone: dyn_hood_l(ComponentPose=Local To Component(LocalPose=@n2), Alpha=1, ExternalForce="0,0,0") {
    angularDampingOverride=0.9999
    angularSpringConstant=1500
    bAngularSpring=true
    bOverrideAngularDamping=true
    bUseGravityOverride=true
    gravityScale=0.35
    simulationSpace=RootRelative
    alphaBoolBlend {blendOption=Linear}
    boundBone {boneName=dyn_hood_l}
    gravityOverride {z=-980}
    physicsBodyDefinitions [{
      boundBone {boneName=dyn_hood_l}
      boxExtents=<10 5 10>
      constraintSetup {angularLimitsMax=<0 10 0>, angularLimitsMin=<0 -10 -30>, angularTarget=<1 0 0>}
      localJointOffset=<0 0 -0.1>
     }]
   }
   n4 = Anim Dynamics - Bone: dyn_hood_r(ComponentPose=@n3, Alpha=1, ExternalForce="0,0,0") {
    angularDampingOverride=0.9999
    angularSpringConstant=1500
    bAngularSpring=true
    bOverrideAngularDamping=true
    bUseGravityOverride=true
    gravityScale=0.35
    simulationSpace=RootRelative
    alphaBoolBlend {blendOption=Linear}
    boundBone {boneName=dyn_hood_r}
    gravityOverride {z=-980}
    physicsBodyDefinitions [{
      boundBone {boneName=dyn_hood_r}
      boxExtents=<10 5 10>
      constraintSetup {angularLimitsMax=<0 10 0>, angularLimitsMin=<0 -10 -30>, angularTarget=<1 0 0>}
      localJointOffset=<0 0 -0.1>
     }]
   }
   Output Pose(Result=Component To Local(ComponentPose=@n4)) {layerGroup=None}
  ]}
 EventGraph {nodes [
   n1 = Event OnCharacterCustomizationLoaded()
   > Set Body_Mesh(Body_Mesh=Get Array Item with EFort Custom Part Type(Array=@n1.CustomMeshParts))
   Event BlueprintUpdateAnimation()
  ]}
]
variables [Body_Mesh {type=SkeletalMeshComponent}]
